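# Material function for the VinderTech LiftAxis umbrella that layers a spray decal over incoming material attributes.
# Samples the Spray texture on UV1, offset and scaled about center, using the scale function's 0-1 mask times texture alpha as coverage.
# A static switch can further limit coverage by a selectable channel of the TechArtMask input.
# Coverage blends BaseColor toward the spray color and Roughness, Metallic and Specular toward spray parameters.
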MaterialFunction /BRCosmetics/Gadgets/Assets/VinderTech_Umbrella/Umbrella_LiftAxis/Materials/MF_LiftAxis_AlignSpray
graph {nodes [
  n1 = GetMaterialAttributes(None=FunctionInput() {inputType=FunctionInput_MaterialAttributes}) {attributeGetTypes [
    #1
    #2
    #3
    #4
   ]}
  n2 = VectorParameter() {group=Spray, parameterName=SprayUV_Offset, sortPriority=2, defaultValue=<rgba 0 0 0 1>}
  n3 = MaterialFunctionCall(
   UVs=Add(A=TextureCoordinate() {coordinateIndex=1}, B=ComponentMask(None=@n2.RGB) {g=true, r=true})
   Texture Scale=ScalarParameter() {defaultValue=0.7, group=Spray, parameterName=SprayScale, sliderMin=4, sortPriority=1}
  ) {materialFunction=&ScaleUVsByCenter}
  n4 = TextureSampleParameter2D(UVs=@n3.UVs) {
   group=Spray
   parameterName=Spray
   sortPriority=0
   texture=/BRCosmetics/Gadgets/Assets/VinderTech_Umbrella/Umbrella_LiftAxis/Textures/T_Umbrella_LifeAxis_Spray
  }
  n5 = Multiply(
   A=Multiply(A=@n3.0-1 mask, B=@n4.A)
   B=StaticSwitch(True=NamedRerouteUsage(), False=Constant() {r=1}, Value=StaticBoolParameter() {group=Spray, parameterName=UseSprayMask, sortPriority=7})
  )
  n6 = SetMaterialAttributes!(
   MaterialAttributes=@n1.MaterialAttributes
   BaseColor=LinearInterpolate(A=@n1.BaseColor, B=@n4.RGB, Alpha=@n5)
   Roughness=LinearInterpolate(A=@n1.Roughness, B=ScalarParameter() {defaultValue=0.5, group=Spray, parameterName=SprayRoughness, sortPriority=4}, Alpha=@n5)
   Metallic=LinearInterpolate(A=@n1.Metallic, B=ScalarParameter() {group=Spray, parameterName=SprayMetallic, sortPriority=5}, Alpha=@n5)
   Specular=LinearInterpolate(A=@n1.Specular, B=ScalarParameter() {defaultValue=0.5, group=Spray, parameterName=SpraySpecular, sortPriority=3}, Alpha=@n5)
  ) {attributeSetTypes [
    #1
    #2
    #3
    #4
   ]}
  FunctionOutput(None=@n6)
  NamedRerouteDeclaration(None=ChannelMaskParameter(None=FunctionInput() {inputName=TechArtMask, sortPriority=1}) {group=Spray, parameterName=SprayMaskChannel, sortPriority=8}.R) {name=SprayMask, nodeColor=<rgba 0 0.9297 1 1>}
 ]}
